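; Event location actor for ToyFox that controls a map icon and player world markers.
; Server-side events toggle the special actor component for the map icon and mark the actor through each player's marker component.
; On cosmetic-capable instances, BeginPlay gets the giant pawn from the game state component, briefly delaying if its ability system is missing, then awaits the intro AlmostCombined gameplay event to cancel local markers.
; EndPlay ends that async wait.
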(event EventBeginPlay
  (bind _returnvalue (Fortnite|Athena|GameState|GetGameStateAthena))
  ;[0 0.15 0.01813 0.5] BeginPlay: On the client, wait for the giant to almost be combined so that the marker icon can be removed then.
  (|CallParentFunction)
  (Networking|CanExecuteCosmeticEvents
    (:True
      (Utilities|IsValid _returnvalue
        (:Is%20Valid
          (bind _returnvalue_2 (Actor|GetComponentbyClass _returnvalue "/ToyFoxGameplay/Gameplay/B_ToyFoxGameStateComponent.B_ToyFoxGameStateComponent_C"))
          (Utilities|IsValid _returnvalue_2
            (:Is%20Valid
              (Utilities|IsValid (Giant|GetGiantPawn _returnvalue_2)
                (:Is%20Valid
                  (Utilities|IsValid (Ability|GetAbilitySystemComponent (Giant|GetGiantPawn _returnvalue_2))
                    (:Is%20Valid
                      _)
                    (:Is%20Not%20Valid
                      (Utilities|FlowControl|Delay 0.2) ; Wait for the giant to replicate in.
                    ))
                  (bind _asynctaskproxy (Ability|Async|WaitGameplayEventToActor (Giant|GetGiantPawn _returnvalue_2) "(TagName=\"ToyFox.Event.Ability.Intro.AlmostCombined\")" true)
                    (:then
                      (Variables|Transient|SetAsyncAction_WaitAlmostCombined _asynctaskproxy))
                    (:EventReceived
                      (CallFunction|UnmarkActorforLocalPlayers))))
                (:Is%20Not%20Valid
                  _)))
            (:Is%20Not%20Valid
              _)))))))

(event Custom|SetMapIconEnable
  (params (bEnable :type Boolean))
  (bind _athenaspecialactor (Variables|Default|GetAthenaSpecialActor))
  ;[0.06271 0 0.15 0.5] SetActiveMapIconEnable: Enable or disable the map icon.
  (Networking|SwitchHasAuthority
    (:Authority
      (Networking|FlushNetDormancy)
      (if bEnable
        (Components|Activation|Activate _athenaspecialactor false)
        (else
          (Components|Activation|Deactivate _athenaspecialactor))))))

(event Custom|MarkActorForPlayers
  ;[0.15 0 0.129 0.5] Mark the actor for all players.
  (Networking|SwitchHasAuthority
    (:Authority
      (for _array_element (Fortnite|GetAllFortPlayerControllers true true)
        (Utilities|IsValid _array_element
          (:Is%20Valid
            (bind _returnvalue (Actor|GetComponentbyClass _array_element "/Script/FortniteGame.AthenaMarkerComponent"))
            (Utilities|IsValid _returnvalue
              (:Is%20Valid
                (MarkerComponent|MarkActoronServer _returnvalue self true true)))))))))

(event Custom|UnmarkActorForLocalPlayers
  (bind _self self)
  ;[0.04312 0.1409 0.15 0.5] Unmark this actor for local players.
  (Networking|CanExecuteCosmeticEvents
    (:True
      (for _array_element (Fortnite|GetLocalFortPlayerControllers true true)
        (Utilities|IsValid _array_element
          (:Is%20Valid
            (bind _returnvalue (Actor|GetComponentbyClass _array_element "/Script/FortniteGame.AthenaMarkerComponent"))
            (Utilities|IsValid _returnvalue
              (:Is%20Valid
                (Utilities|IsValid (MarkerComponent|FindMarkerbyMarkedActor _returnvalue _self)
                  (:Is%20Valid
                    (Class|FortWorldMarker|CancelMarker (MarkerComponent|FindMarkerbyMarkedActor _returnvalue _self))))))))))))

(event EventEndPlay
  (params (EndPlayReason :type EEndPlayReason%20Enum :default "Destroyed"))
  ;[0.1198 0.05412 0 0.5] EndPlay: Clean up action.
  (Utilities|FlowControl|Sequence
    (:then_0
      (bind _asyncaction_waitalmostcombined (Variables|Transient|GetAsyncAction_WaitAlmostCombined)
        (:then
          (Ability|Async|EndAction _asyncaction_waitalmostcombined))
        (:else)))
    (:then_1
      (|CallParentFunction EndPlayReason))))

(fn ConstructionScript)
